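; Test Blueprint for an elimination death pop effect, built from trail, graphic and number flipbooks with 2D sounds.
; On BeginPlay it applies a scale override and yaws the graphic toward player 0's pawn.
; The trail sound plays first, then after a delay the trail hides and the graphic plays with its sound.
; The number flipbook hides when finished, though this graph stops it and never plays it.
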(event EventBeginPlay
  (bind (_returnvalue_roll _returnvalue_pitch _returnvalue_yaw) (Math|Rotator|BreakRotator (Math|Rotator|FindLookatRotation (Transformation|GetActorLocation self) (Transformation|GetActorLocation (Pawn|GetControlledPawn (Game|GetPlayerController 0))))))
  (bind _graphic (Variables|Default|Getgraphic))
  (bind _graphic_2 (Variables|Default|Getgraphic))
  (bind _number (Variables|Default|Getnumber))
  (bind _trail (Variables|Default|Gettrail))
  (Transformation|SetWorldScale3D (Variables|Default|GetDefaultSceneRoot) (* "1.000000,1.000000,1.000000" (Variables|Default|GetscaleOverride)))
  (Utilities|FlowControl|Sequence
    (:then_0
      (Sprite|SetFlipbook _number (Variables|Default|GetnumberFlipbook))
      (Components|Flipbook|Stop _number)
      (Transformation|SetWorldRotation (Variables|Default|Getgraphic) (Math|Rotator|MakeRotator 0.0 0.0 (- _returnvalue_yaw 9e+01)) false false))
    (:then_1
      (Sprite|SetFlipbook _graphic (Variables|Default|GetgraphicFlipbook))
      (Components|Flipbook|Stop _graphic))
    (:then_2
      (Sprite|SetFlipbook _trail (Variables|Default|GettrailFlipbook))
      (Components|Flipbook|SetLooping _trail false))
    (:then_3
      (Audio|PlaySound2D (Variables|Default|GettrailAudio))
      (Utilities|FlowControl|Delay (Variables|Default|GetpopDelay))
      (Components|Flipbook|Play _graphic_2)
      (Rendering|SetVisibility _graphic_2 true false)
      (Rendering|SetVisibility (Variables|Default|Gettrail) false false)
      (EventDispatchers|BindEventtoOnFinishedPlaying (Variables|Default|Getnumber) (delegate Custom|numberDone))
      (Audio|PlaySound2D (Variables|Default|GetgraphicAudio)))))

(event Custom|numberDone
  (Rendering|SetVisibility (Variables|Default|Getnumber) false false))

(fn ConstructionScript)
